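# Sound cue for the Brawny Bass glider's thrust loop, played with glider thrust attenuation.
# A mixer blends five looping layers (wind, flap, mech, creak, texture) with a looping branch that adds randomly chosen balloon creaks after random delays.
# Every branch runs through a continuous modulator, so the AnyDirection parameter scales both pitch and volume, with volume ranges varying by layer.
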SoundCue /BRCosmetics/Sounds/Gliders/BrawnyBass/GliderThrustLoop_BrawnyBass_Cue
root=Mixer_0
nodes [
 {id=Delay_0, children [ModulatorContinuous_4], properties {delayMax=4, delayMin=1}}
 {id=Looping_0, children [Delay_0]}
 {
  id=Mixer_0
  children [ModulatorContinuous_0, ModulatorContinuous_1, ModulatorContinuous_2, ModulatorContinuous_3, ModulatorContinuous_5, Looping_0]
  properties {inputVolume [0.5,0.75,0.8,1,0.6,0.6]}
 }
 {
  id=ModulatorContinuous_0
  children [WavePlayer_0]
  properties {
   pitchModulationParams {default=0, maxOutput=1.2, minOutput=1, parameterName=AnyDirection}
   volumeModulationParams {default=0, maxOutput=0.8, minOutput=0.4, parameterName=AnyDirection}
  }
 }
 {
  id=ModulatorContinuous_1
  children [WavePlayer_1]
  properties {
   pitchModulationParams {default=0, maxOutput=1.2, minOutput=1, parameterName=AnyDirection}
   volumeModulationParams {default=0, minOutput=0.4, parameterName=AnyDirection}
  }
 }
 {
  id=ModulatorContinuous_2
  children [WavePlayer_2]
  properties {
   pitchModulationParams {default=0, maxOutput=1.2, minOutput=1, parameterName=AnyDirection}
   volumeModulationParams {default=0, minOutput=0.5, parameterName=AnyDirection}
  }
 }
 {
  id=ModulatorContinuous_3
  children [WavePlayer_14]
  properties {
   pitchModulationParams {default=0, maxOutput=1.2, minOutput=1, parameterName=AnyDirection}
   volumeModulationParams {default=0, minOutput=0.7, parameterName=AnyDirection}
  }
 }
 {
  id=ModulatorContinuous_4
  children [Random_0]
  properties {
   pitchModulationParams {default=0, maxOutput=1.2, minOutput=1, parameterName=AnyDirection}
   volumeModulationParams {default=0, maxOutput=0.8, minOutput=0.4, parameterName=AnyDirection}
  }
 }
 {
  id=ModulatorContinuous_5
  children [WavePlayer_15]
  properties {
   pitchModulationParams {default=0, maxOutput=1.2, minOutput=1, parameterName=AnyDirection}
   volumeModulationParams {default=0, maxOutput=0.8, minOutput=0.4, parameterName=AnyDirection}
  }
 }
 {
  id=Random_0
  children [
   WavePlayer_3
   WavePlayer_4
   WavePlayer_5
   WavePlayer_6
   WavePlayer_7
   WavePlayer_8
   WavePlayer_9
   WavePlayer_10
   WavePlayer_11
   WavePlayer_12
   WavePlayer_13
  ]
  properties {weights [1,1,1,1,1,1,1,1,1,1,1]}
 }
 {id=WavePlayer_0, properties {bLooping=true, soundWaveAssetPtr=&/Glider_BrawnyBass_Thrust_LP_Wind}}
 {id=WavePlayer_1, properties {bLooping=true, soundWaveAssetPtr=&/Glider_BrawnyBass_Thrust_LP_Flap}}
 {id=WavePlayer_10, properties {soundWaveAssetPtr=&/Glider_BrawnyBass_Balloon_Creak_08}}
 {id=WavePlayer_11, properties {soundWaveAssetPtr=&/Glider_BrawnyBass_Balloon_Creak_09}}
 {id=WavePlayer_12, properties {soundWaveAssetPtr=&/Glider_BrawnyBass_Balloon_Creak_10}}
 {id=WavePlayer_13, properties {soundWaveAssetPtr=&/Glider_BrawnyBass_Balloon_Creak_11}}
 {id=WavePlayer_14, properties {bLooping=true, soundWaveAssetPtr=&/Glider_BrawnyBass_Thrust_LP_Creak}}
 {id=WavePlayer_15, properties {bLooping=true, soundWaveAssetPtr=&/Glider_BrawnyBass_Thrust_LP_Texture}}
 {id=WavePlayer_2, properties {bLooping=true, soundWaveAssetPtr=&/Glider_BrawnyBass_Thrust_LP_Mech}}
 {id=WavePlayer_3, properties {soundWaveAssetPtr=&/Glider_BrawnyBass_Balloon_Creak_01}}
 {id=WavePlayer_4, properties {soundWaveAssetPtr=&/Glider_BrawnyBass_Balloon_Creak_02}}
 {id=WavePlayer_5, properties {soundWaveAssetPtr=&/Glider_BrawnyBass_Balloon_Creak_03}}
 {id=WavePlayer_6, properties {soundWaveAssetPtr=&/Glider_BrawnyBass_Balloon_Creak_04}}
 {id=WavePlayer_7, properties {soundWaveAssetPtr=&/Glider_BrawnyBass_Balloon_Creak_05}}
 {id=WavePlayer_8, properties {soundWaveAssetPtr=&/Glider_BrawnyBass_Balloon_Creak_06}}
 {id=WavePlayer_9, properties {soundWaveAssetPtr=&/Glider_BrawnyBass_Balloon_Creak_07}}
]
properties {
 attenuationSettings=&Fort_Player_Glider_Thrust_Athena_Att
 audioPropertiesSheet=&APS_Cosmetic_Glider_Loop
 maxDistance=12500
 soundClassObject=&Fort_Gliders_Land
 volumeMultiplier=0.9
 concurrencySet [null]
}
